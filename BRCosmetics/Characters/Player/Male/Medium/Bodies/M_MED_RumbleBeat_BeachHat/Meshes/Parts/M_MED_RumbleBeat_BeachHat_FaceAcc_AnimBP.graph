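# Character-part AnimBP for the RumbleBeat BeachHat face accessory, adding secondary motion to hair bones.
# Poses copied from parent, body and head meshes are blended per bone, then a nudged hair aim bone feeds a Control Rig.
# The cached result runs through RigidBody simulation, then spring-limited AnimDynamics on bangs, top hair and mullet.
# EventGraph updates crouch-targeting, melee, skydive and gravity/RBAN state, and caches meshes on customization load.
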AnimBlueprint /BRCosmetics/Characters/Player/Male/Medium/Bodies/M_MED_RumbleBeat_BeachHat/Meshes/Parts/M_MED_RumbleBeat_BeachHat_FaceAcc_AnimBP
target_skeleton=&/M_MED_RumbleBeat_BeachHat_FaceAcc_Skeleton
graphs [
 AnimGraph {nodes [
   n1 = RigidBody!(
    ComponentPose=Local To Component(LocalPose=Use cached pose 'Start_Cache'())
    Alpha
    OverrideWorldGravity:Vector="0,0,0"
    ExternalForce
    SimSpaceSettings:Sim Space Settings Structure=(WorldAlpha=0.3,VelocityScaleZ=1,DampingAlpha=1,MaxLinearVelocity=10000,MaxAngularVelocity=10000,MaxLinearAcceleration=11000,MaxAngularAcceleration=11000,ExternalLinearDragV=(X=0,Y=1.5,Z=0),ExternalLinearVelocity=(X=0,Y=0,Z=0),ExternalAngularVelocity=(X=0,Y=0,Z=0))
   ) -> (Pose) {
    bOverrideWorldGravity=true
    overridePhysicsAsset=&/M_MED_RumbleBeat_BeachHat_FaceAcc_Physics
    simulationSpace=BaseBoneSpace
    baseBoneRef {boneName=neck_01}
    simSpaceSettings {maxAngularAcceleration=11000, maxLinearAcceleration=11000, worldAlpha=0.3, externalLinearDragV {y=1.5}}
   }
   n2 = Anim Dynamics - Bone: dyn_bangs01_rt(ComponentPose=@n1, Alpha=1, ExternalForce="0,0,0") {
    angularDampingOverride=1
    angularSpringConstant=15000
    bAngularSpring=true
    bOverrideAngularDamping=true
    bUseGravityOverride=true
    gravityScale=0.35
    simulationSpace=RootRelative
    worldSpaceMinimumScale=0
    boundBone {boneName=dyn_bangs01_rt}
    componentAppliedLinearAccClamp=<4500 4500 4500>
    componentLinearAccScale=<0.45 0.45 0.45>
    componentLinearVelScale=<0.45 0.45 0.45>
    gravityOverride {z=-100}
    physicsBodyDefinitions [{
      boundBone {boneName=dyn_bangs01_rt}
      boxExtents=<33 33 33>
      constraintSetup {angularLimitsMax=<0 5 5>, angularLimitsMin=<0 -5 -5>, angularTarget=<1 0 0>}
      localJointOffset=<2 0 0>
     }]
    relativeSpaceBone {boneName=base_bangs01_rt}
    simSpaceSettings {maxAngularAcceleration=18, maxAngularVelocity=8, simSpaceAngularAlpha=0.47}
   }
   n3 = Anim Dynamics - Bone: dyn_bangs02_rt(ComponentPose=@n2, Alpha=1, ExternalForce="0,0,0") {
    angularDampingOverride=1
    angularSpringConstant=15000
    bAngularSpring=true
    bOverrideAngularDamping=true
    bUseGravityOverride=true
    gravityScale=0.35
    simulationSpace=RootRelative
    worldSpaceMinimumScale=0
    boundBone {boneName=dyn_bangs02_rt}
    componentAppliedLinearAccClamp=<4500 4500 4500>
    componentLinearAccScale=<0.45 0.45 0.45>
    componentLinearVelScale=<0.45 0.45 0.45>
    gravityOverride {z=-600}
    physicsBodyDefinitions [{
      boundBone {boneName=dyn_bangs02_rt}
      boxExtents=<30 30 30>
      constraintSetup {angularLimitsMax=<0 15 20>, angularLimitsMin=<0 -20 -30>, angularTarget=<1 0 0>}
      localJointOffset=<3 0 0>
     }]
    relativeSpaceBone {boneName=base_bangs01_rt}
    simSpaceSettings {maxAngularAcceleration=18, maxAngularVelocity=8, simSpaceAngularAlpha=0.47}
   }
   n4 = Anim Dynamics - Bone: dyn_bangs03_rt(ComponentPose=@n3, Alpha=1, ExternalForce="0,0,0") {
    angularDampingOverride=1
    angularSpringConstant=15000
    bAngularSpring=true
    bOverrideAngularDamping=true
    bUseGravityOverride=true
    gravityScale=0.35
    simulationSpace=RootRelative
    worldSpaceMinimumScale=0
    boundBone {boneName=dyn_bangs03_rt}
    componentAppliedLinearAccClamp=<4500 4500 4500>
    componentLinearAccScale=<0.45 0.45 0.45>
    componentLinearVelScale=<0.45 0.45 0.45>
    gravityOverride {z=-600}
    physicsBodyDefinitions [{
      boundBone {boneName=dyn_bangs03_rt}
      boxExtents=<28 28 28>
      constraintSetup {angularLimitsMax=<0 20 25>, angularLimitsMin=<0 -25 -35>, angularTarget=<1 0 0>}
      localJointOffset=<3.5 0 0>
     }]
    relativeSpaceBone {boneName=base_bangs01_rt}
    simSpaceSettings {maxAngularAcceleration=18, maxAngularVelocity=8, simSpaceAngularAlpha=0.47}
   }
   n5 = Anim Dynamics - Bone: dyn_hairTop_cr(ComponentPose=@n4, Alpha=1, ExternalForce="0,0,0") {
    angularDampingOverride=1
    angularSpringConstant=3500
    bAngularSpring=true
    bOverrideAngularDamping=true
    bUseGravityOverride=true
    gravityScale=0.35
    simulationSpace=BoneRelative
    worldSpaceMinimumScale=0
    boundBone {boneName=dyn_hairTop_cr}
    componentAppliedLinearAccClamp=<2000 2000 2000>
    componentLinearAccScale=<0.25 0.25 0.25>
    componentLinearVelScale=<0.25 0.25 0.25>
    gravityOverride {z=-200}
    physicsBodyDefinitions [{
      boundBone {boneName=dyn_hairTop_cr}
      boxExtents=<30 30 30>
      constraintSetup {angularLimitsMax=<0 2 5>, angularLimitsMin=<0 -8 -5>, angularTarget=<1 0 0>}
      localJointOffset=<1.5 0 0>
     }]
    relativeSpaceBone {boneName=base_hairTop_cr}
    simSpaceSettings {maxAngularAcceleration=18, maxAngularVelocity=8, simSpaceAngularAlpha=0.4}
   }
   n6 = Anim Dynamics - Bone: dyn_mullet_rt(ComponentPose=@n5, Alpha=1, ExternalForce="0,0,0") {
    angularDampingOverride=1
    angularSpringConstant=4500
    bAngularSpring=true
    bOverrideAngularDamping=true
    bUseGravityOverride=true
    gravityScale=0.35
    simulationSpace=BoneRelative
    worldSpaceMinimumScale=0
    boundBone {boneName=dyn_mullet_rt}
    componentAppliedLinearAccClamp=<2500 2500 2500>
    componentLinearAccScale=<0.25 0.25 0.25>
    componentLinearVelScale=<0.25 0.25 0.25>
    gravityOverride {z=-100}
    physicsBodyDefinitions [{
      boundBone {boneName=dyn_mullet_rt}
      boxExtents=<30 30 30>
      constraintSetup {angularLimitsMax=<0 0 10>, angularLimitsMin=<0 -6 -10>, angularTarget=<1 0 0>}
      localJointOffset=<2 0 0>
     }]
    relativeSpaceBone {boneName=base_mullet_rt}
    simSpaceSettings {maxAngularAcceleration=8, maxAngularVelocity=5, simSpaceAngularAlpha=0.6}
   }
   n7 = Anim Dynamics - Bone: dyn_mullet_cr(ComponentPose=@n6, Alpha=1, ExternalForce="0,0,0") {
    angularDampingOverride=1
    angularSpringConstant=4500
    bAngularSpring=true
    bOverrideAngularDamping=true
    bUseGravityOverride=true
    gravityScale=0.35
    simulationSpace=BoneRelative
    worldSpaceMinimumScale=0
    boundBone {boneName=dyn_mullet_cr}
    componentAppliedLinearAccClamp=<2500 2500 2500>
    componentLinearAccScale=<0.25 0.25 0.25>
    componentLinearVelScale=<0.25 0.25 0.25>
    gravityOverride {z=-100}
    physicsBodyDefinitions [{
      boundBone {boneName=dyn_mullet_cr}
      boxExtents=<30 30 30>
      constraintSetup {angularLimitsMax=<0 0 10>, angularLimitsMin=<0 -6 -10>, angularTarget=<1 0 0>}
      localJointOffset=<2 0 0>
     }]
    relativeSpaceBone {boneName=base_mullet_cr}
    simSpaceSettings {maxAngularAcceleration=8, maxAngularVelocity=5, simSpaceAngularAlpha=0.6}
   }
   n8 = Anim Dynamics - Bone: dyn_mullet_lf(ComponentPose=@n7, Alpha=1, ExternalForce="0,0,0") {
    angularDampingOverride=1
    angularSpringConstant=4500
    bAngularSpring=true
    bOverrideAngularDamping=true
    bUseGravityOverride=true
    gravityScale=0.35
    simulationSpace=BoneRelative
    worldSpaceMinimumScale=0
    boundBone {boneName=dyn_mullet_lf}
    componentAppliedLinearAccClamp=<2500 2500 2500>
    componentLinearAccScale=<0.25 0.25 0.25>
    componentLinearVelScale=<0.25 0.25 0.25>
    gravityOverride {z=-100}
    physicsBodyDefinitions [{
      boundBone {boneName=dyn_mullet_lf}
      boxExtents=<30 30 30>
      constraintSetup {angularLimitsMax=<0 0 10>, angularLimitsMin=<0 -6 -10>, angularTarget=<1 0 0>}
      localJointOffset=<2 0 0>
     }]
    relativeSpaceBone {boneName=base_mullet_lf}
    simSpaceSettings {maxAngularAcceleration=8, maxAngularVelocity=5, simSpaceAngularAlpha=0.6}
   }
   Output Pose(Result=Component To Local(ComponentPose=@n8)) {layerGroup=None}
   n9 = Layered blend per bone(
    BasePose=Copy Pose From Mesh() {bUseAttachedParent=true}
    BlendPoses_0=Copy Pose From Mesh(SourceMeshComponent="Get Body Mesh "() -> ("Body Mesh ":Skeletal Mesh Component Object Reference)) {bUseAttachedParent=true}
   ) {layerSetup [{branchFilters [
       {boneName=base_fur_back_00_ct}
       {boneName=dyn_fur_back_00_ct}
       {boneName=base_fur_back_00_lf}
       {boneName=dyn_fur_back_00_lf}
       {boneName=base_fur_back_00_rt}
       {boneName=dyn_fur_back_00_rt}
       {boneName=base_fur_front_00_rt}
       {boneName=dyn_fur_front_00_rt}
       {boneName=base_fur_front_00_lf}
       {boneName=dyn_fur_front_00_lf}
      ]}]}
   n10 = Layered blend per bone(
    BasePose=@n9
    BlendPoses_0=Copy Pose From Mesh(SourceMeshComponent="Get Face Mesh "() -> ("Face Mesh ":Skeletal Mesh Component Object Reference)) {bUseAttachedParent=true}
   ) {layerSetup [{branchFilters [
       {boneName=FACIAL_C_FacialRoot}
       {boneName=FACIAL_L_NasolabialBulge}
       {boneName=FACIAL_R_NasolabialBulge}
       {boneName=FACIAL_L_NasolabialFurrow}
       {boneName=FACIAL_R_NasolabialFurrow}
       {boneName=FACIAL_L_CheekLower}
       {boneName=FACIAL_R_CheekLower}
       {boneName=FACIAL_C_Nose}
       {boneName=FACIAL_C_NoseLower}
       {boneName=FACIAL_L_Nostril}
       {boneName=FACIAL_R_Nostril}
       {boneName=FACIAL_C_NoseTip}
       {boneName=FACIAL_C_MouthUpper}
       {boneName=FACIAL_C_LipUpper}
       {boneName=FACIAL_L_LipUpper}
       {boneName=FACIAL_R_LipUpper}
       {boneName=FACIAL_L_LipUpperOuter}
       {boneName=FACIAL_R_LipUpperOuter}
       {boneName=FACIAL_L_LipCorner}
       {boneName=FACIAL_R_LipCorner}
       {boneName=FACIAL_L_Masseter}
       {boneName=FACIAL_R_Masseter}
       {boneName=FACIAL_C_UnderChin}
       {boneName=FACIAL_L_UnderChin}
       {boneName=FACIAL_R_UnderChin}
       {boneName=FACIAL_C_LowerLipRotation}
       {boneName=FACIAL_C_MouthLower}
       {boneName=FACIAL_C_LipLower}
       {boneName=FACIAL_L_LipLower}
       {boneName=FACIAL_R_LipLower}
       {boneName=FACIAL_L_LipLowerOuter}
       {boneName=FACIAL_R_LipLowerOuter}
       {boneName=FACIAL_C_Jaw}
       {boneName=FACIAL_C_Jawline}
       {boneName=FACIAL_L_ChinSide}
       {boneName=FACIAL_R_ChinSide}
       {boneName=FACIAL_C_Chin}
       {boneName=FACIAL_L_Ear}
       {boneName=FACIAL_R_Ear}
       {boneName=FACIAL_L_Jawline}
       {boneName=FACIAL_R_Jawline}
      ]}]}
   n11 = "Transform (Modify) Bone - Bone: hair_aim_tgt"(ComponentPose=Local To Component(LocalPose=@n10), Alpha=1, Translation="0,5,0", Rotation="0,0,0", Scale="1,1,1") {translationMode=BMM_Additive, translationSpace=BCS_BoneSpace, boneToModify {boneName=hair_aim_tgt}, translation {y=5}}
   n12 = Control Rig!(
    Source=Component To Local(ComponentPose=@n11)
    Alpha
    Target Crouching:bool=Get Target Crouch() -> (Target Crouch:bool)
    Meelee:bool=Get Meelee() -> (Meelee:bool)
    Skydive Aim Alpha:double=Get Skydive Alpha() -> (Skydive Alpha:double)
    PlayingEmote:bool=Get bIsPlayingEmote()
   ) -> (Pose) {controlRigAssetReference {blueprintRigClass=&/M_MED_RumbleBeat_BeachHat_FaceAcc_CtrlRig.M_MED_RumbleBeat_BeachHat_FaceAcc_CtrlRig_C}}
   Save cached pose 'Start_Cache'(Pose=@n12) {cacheName=Start_Cache}
  ]}
 EventGraph {nodes [
   Event BlueprintUpdateAnimation()
   > n1 = "Gravity Override State Function (Parameter Asset) (Nativized)"(
    CharacterPartAnimInstance=Self-Reference()
    GameTimeInSeconds=GetGameTimeInSeconds()
    Parameters=&/M_MED_RumbleBeat_BeachHat_FaceAcc_Parameters
   )
   > Set GravityOverride(GravityOverride=@n1)
   > Set RBAN Settings Out(RBAN Settings Out=@n1.OutRigidBodyAnimNodeParameters)
   > Set Target Crouch(Target Crouch:bool=AND Boolean!(A=Get bIsTargeting(), B=Get bIsCrouching(), C:bool=Get bIsMoving2D()) -> (ReturnValue)) -> (Output_Get:bool)
   > Set Meelee(Meelee:bool=NOT Boolean(A=Get bIsPlayingMeleeAnim())) -> (Output_Get:bool)
   > n2 = Branch(Condition=Get bIsSkydiveFloating())
   > Set Skydive Alpha(Skydive Alpha:double=0.65) -> (Output_Get:double)
   Set Skydive Alpha(execute=@n2.else, Skydive Alpha=0.5)
   Event OnCharacterCustomizationLoaded()
   > n3 = Cast To PlayerPawn_Athena(Object=TryGetPawnOwner())
   > "Set Body Mesh "("Body Mesh ":Skeletal Mesh Component Object Reference=GetSkeletalMeshForPartType(self=@n3)) -> (Output_Get:Skeletal Mesh Component Object Reference)
   > "Set Face Mesh "("Face Mesh ":Skeletal Mesh Component Object Reference=GetSkeletalMeshForPartType(self=@n3, PartType=Head)) -> (Output_Get:Skeletal Mesh Component Object Reference)
  ]}
]
variables [
 "Body Mesh " {type=SkeletalMeshComponent}
 "Face Mesh " {type=SkeletalMeshComponent}
 GravityOverride {type=Vector}
 Meelee {type=boolean}
 RBAN Settings Out {type=FortRigidBodyAnimNodeParameters}
 Skydive Alpha {type=number}
 Target Crouch {type=boolean}
]
